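# Physics asset for the HydraTrumpet Coach gadget prop, a cosmetic animated mesh whose bones are named for Millennium Falcon and TIE Fighter proxies.
# Each of the six bones has one capsule body: a large one on root, tiny ones on offset and main, and mid-sized ones on the three ships.
# Constraints chain root to offset to main, which branches to the ships, all sharing identical swing and twist limits.
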PhysicsAsset /BRCosmetics/Animation/Props/Gadgets_HydraTrumpet_Coach/Meshes/PA_HydraTrumpet_Coach_Gadget
bodies {
 Mesh_Millennium_Falcon_Proxy_LOD0_JNT [{shapeType=Capsule, length=17.892, radius=22.0431, center=<-1.7711 2.5218 0.4178>, rotation=<rot 3.7669 25.6767 -90>}]
 Mesh_Tie_Fighter_Proxy_003_LOD0_JNT [{shapeType=Capsule, length=4.5717, radius=14.6841, center=<-0.0001 0 0>, rotation=<rot 0 -90 -180>}]
 Mesh_Tie_Fighter_Proxy_004_LOD0_JNT [{shapeType=Capsule, length=4.5717, radius=14.6841, center=<-0.0001 0 0>, rotation=<rot 0 -90 180>}]
 main [{shapeType=Capsule, length=0.505, radius=0.505, rotation=<rot 0 -90 0>}]
 offset [{shapeType=Capsule, length=0.505, radius=0.505, rotation=<rot 0 -90 0>}]
 root [{shapeType=Capsule, length=14.0882, radius=106.1342, center=<37.2168 0 0>, rotation=<rot 0 -90 -90>}]
}
constraints [
 {bone1Name=offset, bone2Name=root, swing1LimitDegrees=45, swing2LimitDegrees=45, twistLimitDegrees=45}
 {bone1Name=main, bone2Name=offset, swing1LimitDegrees=45, swing2LimitDegrees=45, twistLimitDegrees=45}
 {bone1Name=Mesh_Millennium_Falcon_Proxy_LOD0_JNT, bone2Name=main, swing1LimitDegrees=45, swing2LimitDegrees=45, twistLimitDegrees=45}
 {bone1Name=Mesh_Tie_Fighter_Proxy_003_LOD0_JNT, bone2Name=main, swing1LimitDegrees=45, swing2LimitDegrees=45, twistLimitDegrees=45}
 {bone1Name=Mesh_Tie_Fighter_Proxy_004_LOD0_JNT, bone2Name=main, swing1LimitDegrees=45, swing2LimitDegrees=45, twistLimitDegrees=45}
]
